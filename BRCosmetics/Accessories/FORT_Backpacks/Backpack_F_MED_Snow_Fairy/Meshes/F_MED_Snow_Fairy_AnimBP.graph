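# Anim Blueprint for the Snow Fairy backpack that poses its wings according to player movement.
# A wingPosition state machine switches base, dive, glide and crouch poses on skydive, parachute and targeting flags, with speed-weighted additive sprint and base poses applied outside skydiving.
# Four wing AnimDynamics nodes take a Perlin-noise falling force computed from pawn height in the EventGraph, but their chain is not wired to the output pose.
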AnimBlueprint /BRCosmetics/Accessories/FORT_Backpacks/Backpack_F_MED_Snow_Fairy/Meshes/F_MED_Snow_Fairy_AnimBP
target_skeleton=&/F_MED_Snow_Fairy_Skeleton
graphs [
 AnimGraph {nodes [
   n1 = Get falling_anim_dyn_up_force_vec3() -> (falling_anim_dyn_up_force_vec3:Vector)
   n2 = Anim Dynamics - Bone: dyn_Wing_T_L(ComponentPose=Local To Component(), Alpha=0.157, ExternalForce=@n1) {
    alpha=0.9
    angularSpringConstant=1500
    bAngularSpring=true
    bGravityOverrideInSimSpace=true
    gravityScale=0.1
    numSolverIterationsPostUpdate=4
    numSolverIterationsPreUpdate=16
    alphaBoolBlend {blendOption=Linear}
    boundBone {boneName=dyn_Wing_T_L}
    componentLinearAccScale=<0.25 0.25 0.25>
    componentLinearVelScale=<0.25 0.25 0.25>
    physicsBodyDefinitions [{
      sphereCollisionRadius=0
      boundBone {boneName=dyn_Wing_T_L}
      boxExtents=<15 5 5>
      constraintSetup {angularTargetAxis=AxisZ, twistAxis=AxisZ, angularLimitsMax=<0 30 30>, angularLimitsMin=<0 -30 -30>, angularTarget=<0 0 1>}
      localJointOffset=<0 0 0.5>
     }]
   }
   n3 = Anim Dynamics - Bone: dyn_Wing_T_R(ComponentPose=@n2, Alpha=0.157, ExternalForce=@n1) {
    alpha=0.9
    angularSpringConstant=1500
    bAngularSpring=true
    bGravityOverrideInSimSpace=true
    gravityScale=0.1
    numSolverIterationsPostUpdate=4
    numSolverIterationsPreUpdate=16
    alphaBoolBlend {blendOption=Linear}
    boundBone {boneName=dyn_Wing_T_R}
    componentLinearAccScale=<0.25 0.25 0.25>
    componentLinearVelScale=<0.25 0.25 0.25>
    physicsBodyDefinitions [{
      sphereCollisionRadius=0
      boundBone {boneName=dyn_Wing_T_R}
      boxExtents=<15 5 5>
      constraintSetup {angularTargetAxis=AxisZ, twistAxis=AxisZ, angularLimitsMax=<0 30 30>, angularLimitsMin=<0 -30 -30>, angularTarget=<0 0 1>}
      localJointOffset=<0 0 -0.5>
     }]
   }
   n4 = Anim Dynamics - Bone: dyn_Wing_B_L(ComponentPose=@n3, Alpha=0.157, ExternalForce=@n1) {
    alpha=0.9
    angularDampingOverride=0.98
    angularSpringConstant=1500
    bAngularSpring=true
    bGravityOverrideInSimSpace=true
    bOverrideAngularDamping=true
    bOverrideLinearDamping=true
    gravityScale=0.1
    linearDampingOverride=0.98
    numSolverIterationsPostUpdate=4
    numSolverIterationsPreUpdate=16
    alphaBoolBlend {blendOption=Linear}
    boundBone {boneName=dyn_Wing_B_L}
    componentLinearAccScale=<0.25 0.25 0.25>
    componentLinearVelScale=<0.25 0.25 0.25>
    physicsBodyDefinitions [{
      sphereCollisionRadius=0
      boundBone {boneName=dyn_Wing_B_L}
      boxExtents=<15 5 5>
      constraintSetup {angularTargetAxis=AxisZ, twistAxis=AxisZ, angularLimitsMax=<0 30 30>, angularLimitsMin=<0 -30 -30>, angularTarget=<0 0 1>}
      localJointOffset=<0 0 0.5>
     }]
   }
   n5 = Anim Dynamics - Bone: dyn_Wing_B_R(ComponentPose=@n4, Alpha=0.157, ExternalForce=@n1) {
    alpha=0.9
    angularDampingOverride=0.98
    angularSpringConstant=1500
    bAngularSpring=true
    bGravityOverrideInSimSpace=true
    bOverrideAngularDamping=true
    bOverrideLinearDamping=true
    gravityScale=0.1
    linearDampingOverride=0.98
    numSolverIterationsPostUpdate=4
    numSolverIterationsPreUpdate=16
    alphaBoolBlend {blendOption=Linear}
    boundBone {boneName=dyn_Wing_B_R}
    componentLinearAccScale=<0.25 0.25 0.25>
    componentLinearVelScale=<0.25 0.25 0.25>
    physicsBodyDefinitions [{
      sphereCollisionRadius=0
      boundBone {boneName=dyn_Wing_B_R}
      boxExtents=<15 5 5>
      constraintSetup {angularTargetAxis=AxisZ, twistAxis=AxisZ, angularLimitsMax=<0 30 30>, angularLimitsMin=<0 -30 -30>, angularTarget=<0 0 1>}
      localJointOffset=<0 0 -0.5>
     }]
   }
   Component To Local(ComponentPose=@n5)
   n6 = MapRangeClamped(Value=Get PawnSpeed2D(), InRangeA=0, InRangeB=549)
   n7 = Snow_Fairy_Sprint_Pose_PoseAsset : sprint!(PoseWeight:float=@n6) -> (Pose:Pose Link Structure) {
    is=title AnimGraphNode_PoseByName
    poseAsset=&/Snow_Fairy_Sprint_Pose_PoseAsset
    poseName=sprint
   }
   n8 = Apply Additive(Base=Use cached pose 'preSprint'(), Additive=@n7, Alpha=0.4) {alphaBoolBlend {blendOption=Linear}}
   n9 = Snow_Fairy_Base_Pose_PoseAsset : base!(PoseWeight:float=@n6) -> (Pose:Pose Link Structure) {
    is=title AnimGraphNode_PoseByName
    poseAsset=&/Snow_Fairy_Base_Pose_PoseAsset
    poseName=base
   }
   Output Pose(Result=Blend Poses by bool(
     BlendPose_0=Use cached pose 'preSprint'()
     BlendPose_1=Apply Additive(Base=@n8, Additive=@n9) {alphaBoolBlend {blendOption=Linear}}
     BlendTime_0=0.3
     BlendTime_1=0.3
     bActiveValue=Get bIsSkyDiving()
    ) {blendType=Linear}) {layerGroup=None}
   AnimGraphNode_StateMachine_0 = wingPosition() {is=title AnimGraphNode_StateMachine}
   Save cached pose 'preSprint'(Pose=@AnimGraphNode_StateMachine_0) {cacheName=preSprint}
  ]}
 wingPosition {
  entry=Base
  path=AnimGraph.AnimGraphNode_StateMachine_0.wingPosition
  ids=id
  nodes [
   AnimStateNode_0 = Base() {is=title AnimStateNode}
   AnimStateNode_1 = Diving() {is=title AnimStateNode}
   AnimStateNode_2 = Floating() {is=title AnimStateNode}
   AnimStateNode_3 = Targeting() {is=title AnimStateNode}
   AnimStateTransitionNode_0 = AnimStateTransitionNode()
   AnimStateTransitionNode_1 = AnimStateTransitionNode()
   AnimStateTransitionNode_2 = AnimStateTransitionNode()
   AnimStateTransitionNode_5 = AnimStateTransitionNode()
   AnimStateTransitionNode_6 = AnimStateTransitionNode()
   AnimStateTransitionNode_7 = AnimStateTransitionNode()
  ]
  links [
   {
    dst=Diving
    rule=AnimGraph.AnimGraphNode_StateMachine_0.wingPosition.AnimStateTransitionNode_0.Transition
    src=Base
    via=AnimStateTransitionNode_0
   }
   {
    dst=Targeting
    rule=AnimGraph.AnimGraphNode_StateMachine_0.wingPosition.AnimStateTransitionNode_6.Transition
    src=Base
    via=AnimStateTransitionNode_6
   }
   {
    dst=Floating
    rule=AnimGraph.AnimGraphNode_StateMachine_0.wingPosition.AnimStateTransitionNode_1.Transition
    src=Diving
    via=AnimStateTransitionNode_1
   }
   {
    dst=Base
    rule=AnimGraph.AnimGraphNode_StateMachine_0.wingPosition.AnimStateTransitionNode_2.Transition
    src=Floating
    via=AnimStateTransitionNode_2
   }
   {
    dst=Diving
    rule=AnimGraph.AnimGraphNode_StateMachine_0.wingPosition.AnimStateTransitionNode_5.Transition
    src=Floating
    via=AnimStateTransitionNode_5
   }
   {
    dst=Base
    rule=AnimGraph.AnimGraphNode_StateMachine_0.wingPosition.AnimStateTransitionNode_7.Transition
    src=Targeting
    via=AnimStateTransitionNode_7
   }
  ]
 }
 Base {
  path=AnimGraph.AnimGraphNode_StateMachine_0.wingPosition.AnimStateNode_0.Base
  nodes [Output Animation Pose(Result=Sequence Player 'Snow_Fairy_Base_Pose'() {sequence=&/Snow_Fairy_Base_Pose})]
 }
 Diving {
  path=AnimGraph.AnimGraphNode_StateMachine_0.wingPosition.AnimStateNode_1.Diving
  nodes [Output Animation Pose(Result=Sequence Player 'Snow_Fairy_Dive_Pose'() {sequence=&/Snow_Fairy_Dive_Pose})]
 }
 Floating {
  path=AnimGraph.AnimGraphNode_StateMachine_0.wingPosition.AnimStateNode_2.Floating
  nodes [Output Animation Pose(Result=Sequence Player 'Snow_Fairy_Glide_Pose'() {sequence=&/Snow_Fairy_Glide_Pose})]
 }
 Targeting {
  path=AnimGraph.AnimGraphNode_StateMachine_0.wingPosition.AnimStateNode_3.Targeting
  nodes [Output Animation Pose(Result=Sequence Player 'Snow_Fairy_Crouch_Pose'() {sequence=&/Snow_Fairy_Crouch_Pose})]
 }
 Transition {
  path=AnimGraph.AnimGraphNode_StateMachine_0.wingPosition.AnimStateTransitionNode_0.Transition
  nodes [Result(bCanEnterTransition=Get bIsSkydiveDiveMode())]
 }
 Transition {
  path=AnimGraph.AnimGraphNode_StateMachine_0.wingPosition.AnimStateTransitionNode_1.Transition
  nodes [Result(bCanEnterTransition=NOT Boolean(A=Get bIsSkydiveDiveMode()))]
 }
 Transition {
  path=AnimGraph.AnimGraphNode_StateMachine_0.wingPosition.AnimStateTransitionNode_2.Transition
  nodes [Result(bCanEnterTransition=Get bIsParachuteOpened())]
 }
 Transition {
  path=AnimGraph.AnimGraphNode_StateMachine_0.wingPosition.AnimStateTransitionNode_5.Transition
  nodes [Result(bCanEnterTransition=Get bIsSkydiveDiveMode())]
 }
 Transition {
  path=AnimGraph.AnimGraphNode_StateMachine_0.wingPosition.AnimStateTransitionNode_6.Transition
  nodes [Result(bCanEnterTransition=Get bIsTargeting())]
 }
 Transition {
  path=AnimGraph.AnimGraphNode_StateMachine_0.wingPosition.AnimStateTransitionNode_7.Transition
  nodes [Result(bCanEnterTransition=NOT Boolean(A=Get bIsTargeting()))]
 }
 EventGraph {nodes [
   Event BlueprintUpdateAnimation()
   > Set deltaTimeX(deltaTimeX:double=0) -> (Output_Get:double)
   n1 = Get PawnLastLocation() -> (PawnLastLocation:Vector)
   n2 = SelectFloat(
    A=Make Perlin Noise and Remap!(Value:float=BreakVector(InVec=@n1).Z, RangeOutMin:float=-25000, RangeOutMax:float=25000) -> (ReturnValue:float)
    B=0
    bPickA=Get bIsSkyDiving()
   )
   n3 = SelectFloat(A=float * float(A=@n2, B=0.5), B=@n2, bPickA=Get bIsParachuteOpened())
   > Set falling_anim_dyn_up_force_vec3(falling_anim_dyn_up_force_vec3:Vector=MakeVector(X=@n3, Y=@n3)) -> (Output_Get:Vector)
  ]}
]
variables [
 deltaTimeX {type=number}
 falling_anim_dyn_up_force_vec3 {type=Vector}
]
